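# Niagara system for the Scavenger character's fire trails and glow, attached to skeletal mesh bones or sockets and tied to a foot-hidden visibility flag.
# Two GPU flipbook sprite emitters blend from following the sampled point to staying behind with age; one follows tighter at higher speed.
# A burst of persistent sprite-and-light particles follows the mesh, using curl-noise magnitude to vary orange-yellow color.
# User.bisFootHidden drives every emitter's VisibilityTag; User.bisHandHidden is unused.
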NiagaraSystem /BRCosmetics/Effects/Fort_Effects/Effects/Characters/Athena_Parts/Scavenger/NS_Scavenger_Fire_Trails
dependencies {
 usedDataInterfaces [
  &NiagaraDataInterfaceArrayFloat3
  &NiagaraDataInterfaceCamera
  &NiagaraDataInterfaceCurve
  &NiagaraDataInterfaceVectorCurve
  &NiagaraDataInterfaceColorCurve
 ]
 usedDynamicInputs []
 usedModules [&SystemState]
}
emitters {
 FireTrails {
  simTarget=GPUComputeSim
  emitterUpdateScript [
   &EmitterState()
   &SpawnRate(SpawnRate=25, Use Spawn Probability=false)
  ]
  particleSpawnScript [
   &FN_SetSpawnGroup(Number of Spawn Groups=2)
   /Niagara/Modules/Spawn/Initialization/InitializeParticle(
    Lifetime=&RandomRangeFloat(Minimum=0.8, Maximum=1)
    Module.Write Color=true
    Module.Write Lifetime=true
    Module.Write Mass=true
    Module.Write Material Random=true
    Module.Write Position=true
    Module.Write Scale=false
    Module.Write SpriteRotation=true
    Module.Write SpriteSize=true
    Sprite Rotation=&NormalizedAngleToDegrees(NormalizedAngle=&RandomRangeFloat(Minimum=0, Maximum=1))
    Sprite Size=&Vector2DFromFloat(Value=&RandomRangeFloat(Minimum=10, Maximum=15))
   )
   /SetVariables(Particles.PosOffset:Vector3f=&RandomVector(Vector Scale=1.1, Random Seed=1)) {isSetParameters=true, module=SetVariables_#1}
   &Color(Color=<rgba 10 0.981 0 0.55>)
   /SetVariables(Particles.SpriteRotationRate=&RandomRangeFloat(Minimum=20, Maximum=30)) {isSetParameters=true, module=SetVariables_#2}
   /SetVariables(Particles.SpriteRotationRate=&Multiply_Float(
     A=$Particles.SpriteRotationRate
     B=&MakeCustomFloatFromBool(
      Bool=&SetBoolByFloatComparison(A=&RandomRangeFloat(Minimum=0, Maximum=1), B=0.5)
      True Float=1
      False Float=-1
     )
    )) {isSetParameters=true, module=SetVariables_#3}
   &DynamicMaterialParameters(
    Index 0 Param 2=&RandomRangeFloat(Minimum=0, Maximum=4)
    Index 0 Param 4=2.5
    Module.Param0WriteEnabled=false
    Module.Param1WriteEnabled=true
    Module.Param2WriteEnabled=false
    Module.Param3WriteEnabled=true
   ) {module=DynamicMaterialParameters001}
  ]
  particleUpdateScript [
   &ParticleState()
   &SkeletalMeshLocation(
    Apply Sampled Orientation to Mesh Particles=true
    Bone / Socket Index=&SpawnGroup()
    Bone Sampling Mode=::NewEnumerator4
    Inherit Velocity=false
    Interpolate Mesh Sampling=false
    MaskSpawnGroup=false
    Sampled Position Offset=&SelectVectorFromArray(Direct Array Index=$Particles.SpawnGroup, Vector Selection Array=DataInterface() {propertyValues {FloatData [<5 0 0>, <-5 0 0>]}})
    Sampled Position Offset Coordinate Space:ENiagaraMeshTransforms=::NewEnumerator3
    Skeletal Mesh=$User.NewNiagaraDataInterfaceSkeletalMesh
    Write Sampled Mesh Attrributes to Particle Attributes=true
   )
   &self:OffsetToCamera(Amount:NiagaraFloat=&FloatFromCurve(
     FloatCurve=DataInterface() {propertyValues {Curve {keys [{tangentMode=RCTM_User, time=0, value=0}, {tangentMode=RCTM_User, time=1, value=1}]}}}
     Scale Curve=15
    )) {module=OffsetToCamera}
   &self:OffsetToCamera(Amount=15) {enabled=false, module=OffsetToCamera001}
   /SetVariables(Particles.SpriteSize=&Multiply_Vector2(
     A=$Particles.Initial.SpriteSize
     B=&Vector2DFromFloat(Value=&FloatFromCurve(FloatCurve=DataInterface() {propertyValues {Curve {keys [{tangentMode=RCTM_User, time=0, value=2}, {tangentMode=RCTM_User, time=1, value=1}]}}}))
    )) {isSetParameters=true, module=SetVariables_#4}
   &SpriteRotationRate(Rotation Rate=$Particles.SpriteRotationRate)
   &SubUVAnimation(End Frame=29)
   &ScaleColor(
    Scale Alpha=&Multiply_Float(
     A=1
     B=&FloatFromCurve(FloatCurve=DataInterface() {propertyValues {Curve {keys [
          <at 0 0 RCIM_Cubic>
          {arriveTangent=-0.0116, interpMode=RCIM_Cubic, leaveTangent=-0.0116, tangentMode=RCTM_User, time=0.5, value=1}
          {arriveTangent=-2.686, interpMode=RCIM_Cubic, leaveTangent=-2.686, tangentMode=RCTM_User, time=0.6746, value=0.6742}
          {interpMode=RCIM_Cubic, leaveTangent=-2.371, time=1, value=0}
         ]}}})
    )
    Scale RGB=&VectorFromCurve(VectorCurve=DataInterface() {propertyValues {
       XCurve {keys [<at 0.3242 3 RCIM_Cubic>,<at 0.9878 0.9731>]}
       YCurve {keys [<at 0.3242 2.8 RCIM_Cubic>,<at 1 0.9731>]}
       ZCurve {keys [<at 0.3169 2.6605 RCIM_Cubic>,<at 1 0.9731>]}
      }})
    ScaleA=true
    ScaleRGB=true
   ) {module=ScaleColor001}
   &DynamicMaterialParameters(
    Index 0 Param 1=&FloatFromCurve(FloatCurve=DataInterface() {propertyValues {Curve {keys [<at 0 1>]}}})
    Index 0 Param 2=0.5
    Index 0 Param 3=&FloatFromCurve(FloatCurve=DataInterface() {propertyValues {Curve {keys [<at 0 0.5 RCIM_Cubic>, {arriveTangent=1.5, interpMode=RCIM_Cubic, leaveTangent=1.5, time=0.5265, value=0.7148}, <at 1 2>]}}})
    Module.Param0WriteEnabled=true
    Module.Param1WriteEnabled=false
    Module.Param2WriteEnabled=true
    Module.Param3WriteEnabled=false
   )
   &Drag(Use Linear Drag=true, Use Rotational Drag=false) {enabled=false}
   /SetVariables(Particles.Velocity=&Add_Vector(
     A=$Particles.Velocity
     B=&Multiply_VectorByFloat(
      Vector=&MakeVector(X=0, Y=0, Z=&OneMinusFloat(Float=$Particles.NormalizedAge))
      Float=&Multiply_Float(A=700, B=$Engine.DeltaTime)
     )
    )) {enabled=false, isSetParameters=true, module=SetVariables_#5}
   &SolveForcesAndVelocity(Clamp Velocity=false, Limit Acceleration=false) {enabled=false}
   /SetVariables(Particles.Position=&Lerp_Vector(
     A=&Add_Vector(A=$Output.SkeletalMeshLocation.SampledPosition, B=$Particles.PosOffset)
     B=Linked() {linkedVariable {name=Particles.Position, type=NiagaraPosition}}
     Alpha=$Particles.NormalizedAge
    )) {isSetParameters=true, module=SetVariables_#6}
   /SetVariables(Particles.OldPosition:Vector3f=&MakeCustomVectorFromBool(
     Bool=&SetBoolByFloatComparison(A=$Particles.Age, B=0.0333, Comparison Type=::NewEnumerator6)
     True Vector=$Output.SkeletalMeshLocation.SampledPosition
     False Vector=Linked() {linkedVariable {name=Particles.Position, type=NiagaraPosition}}
    )) {enabled=false, isSetParameters=true, module=SetVariables_#7}
   /SetVariables(Particles.RenderVelocity:Vector3f=&Multiply_VectorByFloat(
     Vector=&Subtract_Vector(A=Linked() {linkedVariable {name=Particles.Position, type=NiagaraPosition}}, B=$Particles.OldPosition)
     Float=30
    )) {enabled=false, isSetParameters=true, module=SetVariables_#8}
   /SetVariables(Particles.VisibilityTag=&MakeCustomIntFromBool(Bool=$User.bisFootHidden, True INT=1, False INT=0)) {isSetParameters=true, module=SetVariables_#9}
  ]
  rendererClasses [&NiagaraSpriteRendererProperties]
  renderers [
   {rendererClass=&NiagaraSpriteRendererProperties}
   {rendererClass=&NiagaraSpriteRendererProperties, rendererIndex=1}
  ]
 }
 FireTrails001 {
  simTarget=GPUComputeSim
  emitterUpdateScript [
   &EmitterState()
   &SpawnRate(SpawnRate=10, Use Spawn Probability=false)
  ]
  particleSpawnScript [
   &InitializeParticle(
    Lifetime=&RandomRangeFloat(Minimum=0.8, Maximum=1)
    Module.Write Color=true
    Module.Write Lifetime=true
    Module.Write Mass=true
    Module.Write Material Random=true
    Module.Write Position=true
    Module.Write Scale=false
    Module.Write SpriteRotation=true
    Module.Write SpriteSize=true
    Sprite Rotation=&NormalizedAngleToDegrees(NormalizedAngle=&RandomRangeFloat(Minimum=0, Maximum=1))
    Sprite Size=&Vector2DFromFloat(Value=&RandomRangeFloat(Minimum=10, Maximum=15))
   )
   /SetVariables(Particles.PosOffset:Vector3f=&Add_Vector(A=&RandomVector(Vector Scale=1.1), B=<0 -5 -10>)) {isSetParameters=true, module=SetVariables_#1}
   &Color(Color=<rgba 10 0.981 0 0.55>)
   /SetVariables(Particles.SpriteRotationRate=&RandomRangeFloat(Minimum=20, Maximum=30)) {isSetParameters=true, module=SetVariables_#2}
   /SetVariables(Particles.SpriteRotationRate=&Multiply_Float(
     A=$Particles.SpriteRotationRate
     B=&MakeCustomFloatFromBool(
      Bool=&SetBoolByFloatComparison(A=&RandomRangeFloat(Minimum=0, Maximum=1), B=0.5)
      True Float=1
      False Float=-1
     )
    )) {isSetParameters=true, module=SetVariables_#3}
   &DynamicMaterialParameters(
    Index 0 Param 2=&RandomRangeFloat(Minimum=0, Maximum=4)
    Index 0 Param 4=2.5
    Module.Param0WriteEnabled=false
    Module.Param1WriteEnabled=true
    Module.Param2WriteEnabled=false
    Module.Param3WriteEnabled=true
   ) {module=DynamicMaterialParameters001}
   &FN_SetSpawnGroup(Number of Spawn Groups=2)
  ]
  particleUpdateScript [
   &ParticleState()
   &SkeletalMeshLocation(
    Apply Sampled Orientation to Mesh Particles=true
    Bone / Socket Index=$Particles.SpawnGroup
    Bone Sampling Mode=::NewEnumerator4
    Inherit Velocity=false
    Interpolate Mesh Sampling=false
    MaskSpawnGroup=false
    Sampled Position Offset=&SelectVectorFromArray(
     Direct Array Index=$Particles.SpawnGroup
     Vector Selection Array=DataInterface() {propertyValues {FloatData [<5 0 0>, <-5 0 0>, <25 0 0>, <-25 0 0>]}}
    )
    Sampled Position Offset Coordinate Space:ENiagaraMeshTransforms=::NewEnumerator3
    Skeletal Mesh=$User.NewNiagaraDataInterfaceSkeletalMesh
    Write Sampled Mesh Attrributes to Particle Attributes=true
   )
   &self:OffsetToCamera(Amount:NiagaraFloat=&FloatFromCurve(
     FloatCurve=DataInterface() {propertyValues {Curve {keys [{tangentMode=RCTM_User, time=0, value=0}, {tangentMode=RCTM_User, time=1, value=1}]}}}
     Scale Curve=10
    )) {module=OffsetToCamera}
   /SetVariables(Particles.SpriteSize=&Multiply_Vector2(
     A=$Particles.Initial.SpriteSize
     B=&Vector2DFromFloat(Value=&FloatFromCurve(FloatCurve=DataInterface() {propertyValues {Curve {keys [{tangentMode=RCTM_User, time=0, value=2}, {tangentMode=RCTM_User, time=1, value=1}]}}}))
    )) {isSetParameters=true, module=SetVariables_#4}
   &SpriteRotationRate(Rotation Rate=$Particles.SpriteRotationRate)
   &SubUVAnimation(End Frame=30)
   &ScaleColor(
    Scale Alpha=&Multiply_Float(
     A=1
     B=&FloatFromCurve(FloatCurve=DataInterface() {propertyValues {Curve {keys [
          <at 0 0 RCIM_Cubic>
          {arriveTangent=-0.0116, interpMode=RCIM_Cubic, leaveTangent=-0.0116, tangentMode=RCTM_User, time=0.5, value=1}
          {arriveTangent=-2.686, interpMode=RCIM_Cubic, leaveTangent=-2.686, tangentMode=RCTM_User, time=0.6746, value=0.6742}
          {interpMode=RCIM_Cubic, leaveTangent=-2.371, time=1, value=0}
         ]}}})
    )
    Scale RGB=&VectorFromCurve(VectorCurve=DataInterface() {propertyValues {
       XCurve {keys [<at 0.3242 3 RCIM_Cubic>,<at 0.9878 0.9731>]}
       YCurve {keys [<at 0.3242 2.8 RCIM_Cubic>,<at 1 0.9731>]}
       ZCurve {keys [<at 0.3169 2.6605 RCIM_Cubic>,<at 1 0.9731>]}
      }})
    ScaleA=true
    ScaleRGB=true
   ) {module=ScaleColor001}
   &DynamicMaterialParameters(
    Index 0 Param 1=&FloatFromCurve(FloatCurve=DataInterface() {propertyValues {Curve {keys [<at 0 1>]}}})
    Index 0 Param 2=0.5
    Index 0 Param 3=&FloatFromCurve(FloatCurve=DataInterface() {propertyValues {Curve {keys [<at 0 0.5 RCIM_Cubic>, {arriveTangent=1.5, interpMode=RCIM_Cubic, leaveTangent=1.5, time=0.5265, value=0.7148}, <at 1 2>]}}})
    Module.Param0WriteEnabled=true
    Module.Param1WriteEnabled=false
    Module.Param2WriteEnabled=true
    Module.Param3WriteEnabled=false
   )
   &Drag(Use Linear Drag=true, Use Rotational Drag=false) {enabled=false}
   /SetVariables(Particles.Velocity=&Add_Vector(
     A=$Particles.Velocity
     B=&Multiply_VectorByFloat(
      Vector=&MakeVector(X=0, Y=0, Z=&OneMinusFloat(Float=$Particles.NormalizedAge))
      Float=&Multiply_Float(A=700, B=$Engine.DeltaTime)
     )
    )) {enabled=false, isSetParameters=true, module=SetVariables_#5}
   &SolveForcesAndVelocity(Clamp Velocity=false, Limit Acceleration=false)
   /SetVariables(Particles.OldPosition:Vector3f=&MakeCustomVectorFromBool(
     Bool=&SetBoolByFloatComparison(A=$Particles.Age, B=0.0333, Comparison Type=::NewEnumerator6)
     True Vector=$Output.SkeletalMeshLocation.SampledPosition
     False Vector=Linked() {linkedVariable {name=Particles.Position, type=NiagaraPosition}}
    )) {enabled=false, isSetParameters=true, module=SetVariables_#7}
   /SetVariables(Particles.Position=&Lerp_Vector(
     A=&Add_Vector(A=$Output.SkeletalMeshLocation.SampledPosition, B=$Particles.PosOffset)
     B=Linked() {linkedVariable {name=Particles.Position, type=NiagaraPosition}}
     Alpha=&Multiply_Float(
      A=$Particles.NormalizedAge
      B=&FloatFromCurve(
       FloatCurve=DataInterface() {propertyValues {Curve {keys [<at 0 1>,<at 2500 0.25>]}}}
       CurveIndex=&VectorLength(Vector=$Output.SkeletalMeshLocation.SampledVelocity)
      )
     )
    )) {isSetParameters=true, module=SetVariables_#6}
   /SetVariables(Particles.RenderVelocity:Vector3f=&Multiply_VectorByFloat(
     Vector=&Subtract_Vector(A=Linked() {linkedVariable {name=Particles.Position, type=NiagaraPosition}}, B=$Particles.OldPosition)
     Float=30
    )) {enabled=false, isSetParameters=true, module=SetVariables_#8}
   /SetVariables(Particles.VisibilityTag=&MakeCustomIntFromBool(Bool=$User.bisFootHidden, True INT=1, False INT=0)) {isSetParameters=true, module=SetVariables_#9}
  ]
  rendererClasses [&NiagaraSpriteRendererProperties]
  renderers [
   {rendererClass=&NiagaraSpriteRendererProperties}
   {rendererClass=&NiagaraSpriteRendererProperties, rendererIndex=1}
  ]
 }
 Lights {
  emitterUpdateScript [
   &EmitterState(
    Inactive Response=::NewEnumerator1
    Life Cycle Mode=::NewEnumerator1
    Loop Behavior=::NewEnumerator1
    Loop Duration Mode=::NewEnumerator1
    UseLoopDelay=false
   )
   &SpawnBurst_Instantaneous(Spawn Count=4, Use Spawn Probability=false)
  ]
  particleSpawnScript [
   &InitializeParticle(
    Module.Write Color=true
    Module.Write Lifetime=false
    Module.Write Mass=true
    Module.Write Material Random=true
    Module.Write Position=false
    Module.Write Scale=false
    Module.Write SpriteRotation=true
    Module.Write SpriteSize=true
    Position=&SimulationPosition()
    Sprite Size=<50 50>
   )
   /SetVariables(Particles.LightRadius=40) {isSetParameters=true}
   /SetVariables() {isSetParameters=true}
   &FN_SetSpawnGroup(Number of Spawn Groups=2)
  ]
  particleUpdateScript [
   &ParticleState(Kill Particles When Lifetime Has Elapsed=false, Let Infinitely Lived Particles Die When Emitter Deactivates=true)
   &SkeletalMeshLocation(
    Bone / Socket Index=&SpawnGroup()
    Bone Sampling Mode=::NewEnumerator4
    Inherit Velocity=false
    MaskSpawnGroup=false
    Sampled Position Offset=&SelectVectorFromArray(
     Direct Array Index=$Particles.SpawnGroup
     Vector Selection Array=DataInterface() {propertyValues {FloatData [<0 0 0>, <0 0 0>, <25 0 0>, <-25 0 0>]}}
    )
    Sampled Position Offset Coordinate Space:ENiagaraMeshTransforms=::NewEnumerator3
    Skeletal Mesh=$User.NewNiagaraDataInterfaceSkeletalMesh
   )
   &self:OffsetToCamera(Amount=10) {module=OffsetToCamera}
   /Niagara/Modules/Update/Forces/CurlNoiseForce(Noise Quality / Cost=::NewEnumerator0, Noise Strength=1, Pan Noise=true, Pan Noise Field=<0 0 2.5>)
   &SolveForcesAndVelocity(Clamp Velocity=false, Limit Acceleration=false, Write to Intrinsic Properties=false, Write to Presolve Properties=false)
   /SetVariables(Particles.Color=&ColorFromCurve(
     ColorCurve=DataInterface() {propertyValues {
       AlphaCurve {keys [<at 0 1 RCIM_Cubic>]}
       BlueCurve {keys [<at 0 0>,<at 0.4 0 RCIM_Cubic>,<at 0.6 0>,<at 1 0>]}
       GreenCurve {keys [<at 0 1.2425>,<at 0.4 0.2462 RCIM_Cubic>,<at 0.6 0.2462>,<at 1 1.2425>]}
       RedCurve {keys [<at 0 1.5>,<at 0.4 1 RCIM_Cubic>,<at 0.6 1>,<at 1 1.5>]}
      }}
     CurveIndex=&VectorLength(Vector=$Output.CurlNoiseForce.CurlNoiseForce)
    )) {isSetParameters=true}
   /SetVariables(Particles.Color=&MultiplyLinearColor(Linear Color=$Particles.Color, Scale Factor=&MakeVector4_FromFloat(Float=2))) {isSetParameters=true}
   /SetVariables(Particles.VisibilityTag=&MakeCustomIntFromBool(Bool=$User.bisFootHidden, True INT=1, False INT=0)) {isSetParameters=true, module=SetVariables_#9}
  ]
  rendererClasses [&NiagaraSpriteRendererProperties, &NiagaraLightRendererProperties]
  renderers [
   {rendererClass=&NiagaraSpriteRendererProperties}
   {rendererClass=&NiagaraLightRendererProperties, rendererIndex=1}
  ]
 }
}
summary {
 systemName=NS_Scavenger_Fire_Trails
 emitters [
  {emitterName=FireTrails, simTarget=GPUComputeSim, rendererClasses [&NiagaraSpriteRendererProperties]}
  {emitterName=Lights, rendererClasses [&NiagaraSpriteRendererProperties, &NiagaraLightRendererProperties]}
  {emitterName=FireTrails001, simTarget=GPUComputeSim, rendererClasses [&NiagaraSpriteRendererProperties]}
 ]
 userVariables [
  User.bisFootHidden {type=NiagaraBool}
  User.bisHandHidden {type=NiagaraBool}
  User.NewNiagaraDataInterfaceSkeletalMesh {
   type=NiagaraDataInterfaceSkeletalMesh
   defaultValue {
    struct=NiagaraExt_VariableValue_DataInterface
    value {dataInterface=&self:NiagaraDataInterfaceSkeletalMesh_0}
   }
  }
 ]
}
